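# Niagara thruster exhaust effect for the Galactic Female glider's deploy, drawing blue engine flames on the glider's skeletal mesh.
# Mesh flames, glow and energy sprites, and sparks map particle ID modulo 4 onto four sockets, with per-engine size arrays, alongside a burst boost light.
# System thrust intensity scales flames, sparks and energy.
# User deploy, front-end preview and ForwardDot inputs adjust visibility, and time of day tints energy brightness.
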NiagaraSystem /BRCosmetics/Gadgets/Assets/VinderTech_GliderChute/Glider_Galactic_Female/FX/NS_Glider_Galactic_Female_Deploy2
dependencies {
 usedDataInterfaces [
  &NiagaraDataInterfaceArrayFloat3
  &NiagaraDataInterfaceArrayFloat
  &NiagaraDataInterfaceCurve
  &NiagaraDataInterfaceColorCurve
  &NiagaraDataInterfaceVector2DCurve
 ]
 usedDynamicInputs []
 usedModules [
  &self:SystemSpawnScript.SystemScriptSource.SystemScriptGraph.NiagaraNodeAssignment_0.SetVariables_#1
  &SystemState
  &self:Set_Current_Exhaust
  &self:Set_Trail_Length
  &self:Set_Trail_Opacity
  &self:SystemSpawnScript.SystemScriptSource.SystemScriptGraph.NiagaraNodeAssignment_2.SetVariables_#2
 ]
}
emitters {
 Boost_light {
  emitterUpdateScript [
   &EmitterState()
   &SpawnBurst_Instantaneous(Use Spawn Probability=false, UseLoopCountLimit=false)
  ]
  particleSpawnScript [&InitializeParticle(Color=<rgba 0 1.8146 10 50>, UsePositionOffset=false, Write Material Random=true)]
  particleUpdateScript [
   &ParticleState()
   &SkeletalMeshLocation(Bone Sampling Mode=::NewEnumerator4, MaskSpawnGroup=false, Sampled Position Offset=<0 150 0>, Skeletal Mesh=$User.SkeletalMesh)
   &ScaleColor(
    Scale Alpha=&FloatFromCurve(FloatCurve=DataInterface() {propertyValues {Curve {keys [
         {arriveTangent=4.8444, interpMode=RCIM_Cubic, leaveTangent=4.8444, tangentMode=RCTM_User, time=0, value=1}
         {arriveTangent=-1, interpMode=RCIM_Cubic, leaveTangent=-4.4849, tangentMode=RCTM_Break, time=0.1, value=1.5}
         {interpMode=RCIM_Cubic, leaveTangent=-2.1599, time=1, value=0}
        ]}}})
    ScaleA=true
    ScaleRGB=true
   )
  ]
  rendererClasses [&NiagaraSpriteRendererProperties, &NiagaraLightRendererProperties]
  renderers [
   {rendererClass=&NiagaraSpriteRendererProperties}
   {rendererClass=&NiagaraLightRendererProperties, rendererIndex=1}
  ]
 }
 EngineFlames002 {
  emitterSpawnScript [&FN_PreserveSkydiveUserVariables() {module=PreserveUserVariables}]
  emitterUpdateScript [
   &EmitterState()
   &SpawnBurst_Instantaneous(Spawn Count=4, Use Spawn Probability=true)
   &InterpolateOverTime(Rate Of Change=0.1, Target Float Value:NiagaraFloat=$System.Thrust intencity)
  ]
  particleSpawnScript [
   &InitializeParticle(
    Color=<rgba 0 0.9073 5 1>
    Lifetime=1.5
    Mesh Scale=<0.24 0.24 0.24>
    Mesh Scale Mode=::NewEnumerator4
    UsePositionOffset=false
    Write Material Random=true
   ) {module=InitializeParticle001}
   &InitialMeshOrientation(Rotation=<-0.25 0 0>, Use Orientation Vector=true, Use Rotation Vector=true) {module=InitialMeshOrientation001}
   /SetVariables(Particles.FlameCoreScale:Vector3f=&SelectVectorFromArray(
     Direct Array Index=&ReturnParticleIDAttributes()
     Direct Array Mode=::NewEnumerator1
     Vector Selection Array=DataInterface() {propertyValues {FloatData [<1.4 0.75 0.75>, <1.4 0.5 0.5>, <1.4 0.5 0.5>, <1.4 0.54 0.54>]}}
    )) {isSetParameters=true}
   /SetVariables(Particles.FlameCoreOrientation:Quat4f=<0 0 -1 1>) {isSetParameters=true}
   &DynamicMaterialParameters(
    Index 0 Param 3=15
    Module.Param0WriteEnabled=false
    Module.Param1WriteEnabled=false
    Module.Param2WriteEnabled=true
    Module.Param3WriteEnabled=false
   )
   /SetVariables(Particles.SizeZ=2) {isSetParameters=true}
   /SetVariables(Particles.Scale=&Multiply_VectorByFloat(
     Vector=$Particles.Scale
     Float=&SelectFloatFromArray(
      Direct Array Index=&ReturnParticleIDAttributes()
      Direct Array Mode=::NewEnumerator1
      Float Selection Array=DataInterface() {propertyValues {FloatData [2,1.35,1.35,1.8]}}
     )
    )) {isSetParameters=true}
  ]
  particleUpdateScript [
   &ParticleState(Kill Particles When Lifetime Has Elapsed=false, Let Infinitely Lived Particles Die When Emitter Deactivates=true)
   &SkeletalMeshLocation(
    Apply Sampled Orientation to Mesh Particles=true
    Bone / Socket Index=&ModuloInt(A=&ReturnParticleIDAttributes(), B=4)
    Inherit Velocity=false
    MaskSpawnGroup=false
    Mesh Sampling Type=::NewEnumerator5
    Skeletal Mesh=$User.SkeletalMesh
    Socket Sampling Mode=::NewEnumerator5
   )
   &SolveForcesAndVelocity(Clamp Velocity=false, Limit Acceleration=false)
   &Timeline(Rewind Rate=2) {enabled=false}
   /SetVariables(Particles.FlameCoreColor:LinearColor=&MakeLinearColorFromVectorAndFloat(
     "Vector (RGB)"=&MakeVectorFromLinearColorRGB(Color=<rgba 0 2.7219 15 1>)
     "Float (Alpha)"=&Lerp_Float(A=0, B=1, Alpha=$Emitter.InterpolateOverTime.InterpolateOverTime_Float.Moving Average)
    )) {isSetParameters=true}
   /SetVariables(Particles.FlameCoreScale:Vector3f=&Multiply_Vector(
     A=$Particles.Initial.FlameCoreScale
     B=&MakeVector(
      X=&FloatFromCurve(FloatCurve=DataInterface() {propertyValues {Curve {keys [{arriveTangent=3.1325, interpMode=RCIM_Cubic, leaveTangent=3.1325, tangentMode=RCTM_User, time=0, value=0}, <at 1 1 RCIM_Cubic>]}}})
      Y=1
      Z=1
     )
    )) {isSetParameters=true}
   /SetVariables(Particles.FlameCoreScale:Vector3f=&Multiply_Vector(A=$Particles.FlameCoreScale, B=&MakeVector(X=$Particles.SizeZ, Y=1.5, Z=1.5))) {isSetParameters=true}
   &FN_Light_v2(Light Exponent=2, Light Radius=65, LightEnabled:NiagaraBool=$User.bIsFullyDeployed, Position Offset:Vector3f=<0 20 10>, Scale Color=1)
   /SetVariables(Particles.Light Color:LinearColor=&MultiplyLinearColor(
     Linear Color=<rgba 300 47.8006 0 1>
     Scale Factor=&MakeVector4_FromFloat(Float=&Lerp_Float(A=0.001, B=1, Alpha=$User.ForwardDot))
    )) {enabled=false, isSetParameters=true}
   /SetVariables(Particles.FirstPosition:Vector3f=&Add_Vector(A=$Output.SkeletalMeshLocation.SampledPosition, B=<0 -6 0>)) {isSetParameters=true}
   /SetVariables(Particles.secondPosition:Vector3f=&Add_Vector(A=$Output.SkeletalMeshLocation.SampledPosition, B=<0 6 0>)) {isSetParameters=true}
   /SetVariables(Particles.SizeZ:NiagaraFloat=&ClampFloat(Float=&Add_Float(
      A=$Particles.SizeZ
      B=&MakeCustomFloatFromBool(Bool=&SetBoolByFloatComparison(A=$User.ForwardDot, B=0), True Float=0.03, False Float=-1)
     ))) {enabled=false, isSetParameters=true}
   /SetVariables(Particles.SizeZ:NiagaraFloat=&Multiply_Float(A=$System.Thrust intencity, B=3)) {isSetParameters=true}
   &ScaleMeshSize(Scale Factor=&MakeVector(X=0.75, Y=0.75, Z=&Multiply_Float(A=$System.Thrust intencity, B=2)))
   /SetVariables(
    Particles.Extra mesh scale:Vector3f=&Multiply_VectorByFloat(Vector=<1 1 1>, Float=&Multiply_Float(A=$System.Thrust intencity, B=0.45))
    Particles.extra mesh color:LinearColor=<rgba 0 0.1815 1 1>
   ) {isSetParameters=true}
  ]
  rendererClasses [&NiagaraMeshRendererProperties]
  renderers [
   {rendererClass=&NiagaraMeshRendererProperties, rendererIndex=0}
   {rendererClass=&NiagaraMeshRendererProperties, rendererIndex=1}
   {rendererClass=&NiagaraMeshRendererProperties, rendererIndex=2}
  ]
 }
 Glow {
  emitterUpdateScript [
   &EmitterState()
   &SpawnRate(SpawnRate=&Multiply_FloatByInt(Float=10, Integer=4), Use Spawn Probability=false)
  ]
  particleSpawnScript [
   &InitializeParticle(
    Color=<rgba 0 0.5444 3 1>
    Lifetime Max=0.3
    Lifetime Min=0.2
    Lifetime Mode=::NewEnumerator1
    Sprite Size Mode=::NewEnumerator3
    Uniform Sprite Size=&SelectFloatFromArray(
     Direct Array Index=&ReturnParticleIDAttributes()
     Direct Array Mode=::NewEnumerator1
     Float Selection Array=DataInterface() {propertyValues {FloatData [30,18,18,20]}}
    )
    UsePositionOffset=false
    Write Material Random=true
   )
   &CameraOffset(Camera Offset Amount=8)
   &AddVelocity(Velocity=<0 30 0>)
   /SetVariables(Particles.New Vector:Vector3f=<0 1 0>) {isSetParameters=true}
  ]
  particleUpdateScript [
   &SolveForcesAndVelocity(Clamp Velocity=false, Limit Acceleration=false)
   &ParticleState()
   &SkeletalMeshLocation(
    Bone / Socket Index=&ModuloInt(A=&ReturnParticleIDAttributes(), B=4)
    MaskSpawnGroup=false
    Mesh Sampling Type=::NewEnumerator5
    Skeletal Mesh=$User.SkeletalMesh
    Socket Sampling Mode=::NewEnumerator5
   )
   &DynamicMaterialParameters(
    Index 0 Param 1=50
    Module.Param0WriteEnabled=true
    Module.Param1WriteEnabled=true
    Module.Param2WriteEnabled=true
    Module.Param3WriteEnabled=true
   )
   &ScaleColor(
    Scale Alpha=&FloatFromCurve(
     FloatCurve=DataInterface() {propertyValues {Curve {keys [
         {arriveTangent=6.8336, interpMode=RCIM_Cubic, leaveTangent=6.8336, tangentMode=RCTM_User, time=0, value=0}
         {arriveTangent=0.875, interpMode=RCIM_Cubic, leaveTangent=0.875, time=0.2, value=1}
         {arriveTangent=-1.25, interpMode=RCIM_Cubic, leaveTangent=-1.25, time=0.4, value=0.35}
         <at 1 0 RCIM_Cubic>
        ]}}}
     Scale Curve=&MakeCustomFloatFromBool(
      Bool=$User.bIsFrontEndPreview
      True Float=&MakeCustomFloatFromBool(Bool=$User.bIsFullyDeployed, True Float=1, False Float=0)
      False Float=&Lerp_Float(A=0, B=1, Alpha=$User.ForwardDot)
     )
    )
    ScaleA=true
    ScaleRGB=true
   )
   &FN_Light_Attributes(
    Color=<rgba 280 26.1767 0 1>
    Create Color Attribute=true
    Create Exponent Attribute=true
    Create Position Attribute=true
    Create Radius Attribute=true
    Exponent=3
    Position=&Add_Vector(A=$Particles.Initial.SkeletalMeshLocation.SampledPosition, B=<0 18 5>)
    Radius=75
    Scale Color=&FloatFromCurve(FloatCurve=DataInterface() {propertyValues {Curve {keys [
         {arriveTangent=9.0725, interpMode=RCIM_Cubic, leaveTangent=9.0725, tangentMode=RCTM_User, time=0, value=0}
         {arriveTangent=-2.5249, interpMode=RCIM_Cubic, leaveTangent=-2.5249, tangentMode=RCTM_User, time=0.1408, value=0.9847}
         {arriveTangent=-0.0405, interpMode=RCIM_Cubic, leaveTangent=-0.0405, tangentMode=RCTM_User, time=1, value=0}
        ]}}})
   )
  ]
  rendererClasses [&NiagaraSpriteRendererProperties]
  renderers [
   {rendererClass=&NiagaraSpriteRendererProperties}
   {rendererClass=&NiagaraSpriteRendererProperties, rendererIndex=1}
  ]
 }
 Spark {
  emitterUpdateScript [
   &EmitterState()
   &SpawnRate(
    SpawnRate=&Multiply_FloatByInt(Float=&Multiply_Float(A=$System.Thrust intencity, B=4), Integer=4)
    Use Spawn Probability=false
   )
  ]
  particleSpawnScript [
   &InitializeParticle(
    Color=<rgba 0 1.8146 10 1>
    Lifetime Max=0.42
    Lifetime Min=0.18
    Lifetime Mode=::NewEnumerator1
    Sprite Size Max=<2 12>
    Sprite Size Min=<0.85 5>
    Sprite Size Mode=::NewEnumerator2
    UsePositionOffset=false
    Write Material Random=true
   )
   &SkeletalMeshLocation(
    Bone / Socket Index=&ModuloInt(A=&ReturnParticleIDAttributes(), B=4)
    Interpolate Mesh Sampling=false
    MaskSpawnGroup=false
    Mesh Sampling Type=::NewEnumerator5
    Skeletal Mesh=$User.SkeletalMesh
    Socket Sampling Mode=::NewEnumerator5
   )
   &SphereLocation(
    MaskSpawnGroup=false
    Override Seed=false
    Sphere Radius=&SelectFloatFromArray(
     Direct Array Index=&ReturnParticleIDAttributes()
     Direct Array Mode=::NewEnumerator1
     Float Selection Array=DataInterface() {propertyValues {FloatData [13,8,8,10]}}
    )
    Surface Only=false
    UseNonUniformScale=false
    UseOffset=false
    UseSphereOrientation=false
   )
   &AddVelocity(Velocity=&RandomRangeVector(Minimum=<0 250 0>, Maximum=<1 450 1>))
  ]
  particleUpdateScript [
   &SolveForcesAndVelocity(Clamp Velocity=false, Limit Acceleration=false)
   &ParticleState()
   &ScaleSpriteSize(
    Non-Uniform Curve Scale=&MakeCustomVector2DFromBool(
     Bool=$User.bIsFrontEndPreview
     True Vector2D=&MakeCustomVector2DFromBool(Bool=$User.bIsFullyDeployed, True Vector2D=<1 1>, False Vector2D=<0 0>)
     False Vector2D=&Lerp_Vector2(A=<0 0>, B=<1 1>, Alpha=$System.Thrust intencity)
    )
    Non-Uniform Curve Sprite Scale=DataInterface() {propertyValues {
      XCurve {keys [{tangentMode=RCTM_User, time=0, value=1}, {tangentMode=RCTM_User, time=1, value=0.2}]}
      YCurve {keys [{tangentMode=RCTM_User, time=0, value=1}, {tangentMode=RCTM_User, time=1, value=0}]}
     }}
    Scale Sprite Size Mode=::NewEnumerator3
   )
  ]
  rendererClasses [&NiagaraSpriteRendererProperties]
  renderers [{rendererClass=&NiagaraSpriteRendererProperties}]
 }
 energy {
  emitterUpdateScript [
   &EmitterState()
   &SpawnRate(SpawnRate=&Multiply_FloatByInt(Float=40, Integer=4), Use Spawn Probability=false)
  ]
  particleSpawnScript [
   &InitializeParticle(
    Color=<rgba 0 0.9073 5 1>
    Lifetime=0.3
    Sprite Size=&MakeCustomVector2DFromBool(Bool=$User.bIsFrontEndPreview, True Vector2D=<15 50>, False Vector2D=<23 37.5>)
    Sprite Size Mode=::NewEnumerator4
    UsePositionOffset=false
    Write Material Random=true
   )
   &AddVelocity(Velocity=&MakeCustomVectorFromBool(Bool=$User.bIsFrontEndPreview, True Vector=<0 300 1>, False Vector=<1 100 0>))
   &CameraOffset(Camera Offset Amount=2)
   /SetVariables(Particles.Alpha=1, Particles.SizeZ=2) {isSetParameters=true}
  ]
  particleUpdateScript [
   &SolveForcesAndVelocity(Clamp Velocity=false, Limit Acceleration=false)
   &ParticleState()
   &SkeletalMeshLocation(
    Bone / Socket Index=&ModuloInt(A=&ReturnParticleIDAttributes(), B=4)
    MaskSpawnGroup=false
    Mesh Sampling Type=::NewEnumerator5
    Skeletal Mesh=$User.SkeletalMesh
    Socket Sampling Mode=::NewEnumerator5
   )
   &DynamicMaterialParameters(
    Index 1 Param 1=45
    Module.Param0WriteEnabled=true
    Module.Param0WriteEnabled1=true
    Module.Param1WriteEnabled=true
    Module.Param1WriteEnabled1=true
    Module.Param2WriteEnabled=true
    Module.Param2WriteEnabled1=true
    Module.Param3WriteEnabled=true
    Module.Param3WriteEnabled1=true
    Write Parameter Index 1=true
   )
   /SetVariables(Particles.SizeZ:NiagaraFloat=&ClampFloat(Float=&Add_Float(A=$Particles.SizeZ, B=1))) {enabled=false, isSetParameters=true}
   /SetVariables(Particles.Alpha:NiagaraFloat=&ClampFloat(Float=&Add_Float(
      A=$Particles.Alpha
      B=&MakeCustomFloatFromBool(Bool=&SetBoolByFloatComparison(A=$User.ForwardDot, B=0), True Float=0.1, False Float=-0.5)
     ))) {isSetParameters=true}
   &ScaleSpriteSize(
    Scale Factor=&Multiply_Vector2D(
     A=&MakeVector2D(
      X=&FloatFromCurve(
       FloatCurve=DataInterface() {propertyValues {Curve {keys [{arriveTangent=-1.9515, time=0, value=1}, <at 1 0.75>]}}}
       Scale Curve=&Multiply_Float(
        A=&SelectFloatFromArray(
         Direct Array Index=&ReturnParticleIDAttributes()
         Direct Array Mode=::NewEnumerator1
         Float Selection Array=DataInterface() {propertyValues {FloatData [2,1,1,1.5]}}
        )
        B=1
       )
      )
      Y=&FloatFromCurve(
       FloatCurve=DataInterface() {propertyValues {Curve {keys [<at 0 1>,<at 1 3>]}}}
       Scale Curve=&Multiply_Float(A=$Particles.SizeZ, B=1)
      )
     )
     B=&MakeVector2D(X=1, Y=$System.Thrust intencity)
    )
    Scale Sprite Size Mode=::NewEnumerator2
   )
   &ScaleColor(
    Scale Alpha=&Multiply_Float(
     A=&FNTimeOfDayScalar(Night Scale=0.5, Morning Scale=1, Evening Scale=1)
     B=&Lerp_Float(A=0, B=1, Alpha=$User.ForwardDot)
    )
    Scale RGB=&Multiply_VectorByFloat(Vector=<1 1 1>, Float=&FNTimeOfDayScalar(Night Scale=0.4, Morning Scale=1, Evening Scale=0.5))
    ScaleA=true
    ScaleRGB=true
   ) {enabled=false}
   &ScaleColor(
    Scale Alpha=&MakeCustomFloatFromBool(
     Bool=$User.bIsFrontEndPreview
     True Float=&MakeCustomFloatFromBool(Bool=$User.bIsFullyDeployed, True Float=1, False Float=0)
     False Float=&Multiply_Float(
      A=&FNTimeOfDayScalar(Night Scale=0.5, Morning Scale=1, Evening Scale=1)
      B=&Lerp_Float(A=0, B=1, Alpha=$System.Thrust intencity)
     )
    )
    Scale RGB=&Multiply_VectorByFloat(Vector=<1 1 1>, Float=&FNTimeOfDayScalar(Night Scale=1, Morning Scale=1, Evening Scale=0.5))
    ScaleA=true
    ScaleRGB=true
   ) {module=ScaleColor001}
  ]
  rendererClasses [&NiagaraSpriteRendererProperties]
  renderers [{rendererClass=&NiagaraSpriteRendererProperties}]
 }
}
summary {
 systemName=NS_Glider_Galactic_Female_Deploy2
 emitters [
  {emitterName=EngineFlames002, rendererClasses [&NiagaraMeshRendererProperties]}
  {emitterName=energy, rendererClasses [&NiagaraSpriteRendererProperties]}
  {emitterName=Glow, rendererClasses [&NiagaraSpriteRendererProperties]}
  {emitterName=Spark, rendererClasses [&NiagaraSpriteRendererProperties]}
  {emitterName=Boost_light, rendererClasses [&NiagaraSpriteRendererProperties, &NiagaraLightRendererProperties]}
 ]
 userVariables [
  User.bIsFrontEnd {type=NiagaraBool, defaultValue=true}
  User.bIsFrontEndPreview {type=NiagaraBool, defaultValue=true}
  User.bIsFullyDeployed {type=NiagaraBool, defaultValue=true}
  User.ForwardDot {type=NiagaraFloat, defaultValue=1, description=""}
  User.PlayerSpeed {type=NiagaraFloat, defaultValue=0, description=""}
  User.RightDot {type=NiagaraFloat, defaultValue=0, description=""}
  User.SkeletalMesh {
   type=NiagaraDataInterfaceSkeletalMesh
   defaultValue {
    struct=NiagaraExt_VariableValue_DataInterface
    value {dataInterface=&self:NiagaraDataInterfaceSkeletalMesh_0}
   }
  }
 ]
}
